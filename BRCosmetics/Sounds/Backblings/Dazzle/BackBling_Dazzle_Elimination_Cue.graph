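# Sound cue for the Dazzle back bling's kill reaction, played as an elimination sound.
# A Random node picks evenly among five KillReaction waves, then a Modulator applies slight pitch and volume variation.
# It uses close-range pet attenuation, the Fort_Backblings sound class and a reactive back bling property sheet.
# Overridden concurrency allows one instance per owner, stopping the oldest when retriggered.
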SoundCue /BRCosmetics/Sounds/Backblings/Dazzle/BackBling_Dazzle_Elimination_Cue
root=Modulator_0
nodes [
 {id=Modulator_0, children [Random_0], properties {pitchMax=1.02, pitchMin=0.98, volumeMax=0.7, volumeMin=0.6}}
 {id=Random_0, children [WavePlayer_1,WavePlayer_2,WavePlayer_8,WavePlayer_9,WavePlayer_0], properties {weights [1,1,1,1,1]}}
 {id=WavePlayer_0, properties {soundWaveAssetPtr=&/BackBling_Dazzle_KillReaction_05}}
 {id=WavePlayer_1, properties {soundWaveAssetPtr=&/BackBling_Dazzle_KillReaction_01}}
 {id=WavePlayer_2, properties {soundWaveAssetPtr=&/BackBling_Dazzle_KillReaction_02}}
 {id=WavePlayer_8, properties {soundWaveAssetPtr=&/BackBling_Dazzle_KillReaction_03}}
 {id=WavePlayer_9, properties {soundWaveAssetPtr=&/BackBling_Dazzle_KillReaction_04}}
]
properties {
 attenuationSettings=&Fort_Pets_Close_Athena_Att
 audioPropertiesSheet=/BRCosmetics/Sounds/Settings/PropertySheets/Reactive/APS_Cosmetic_Reactive_Backbling
 bOverrideConcurrency=true
 maxDistance=1800.0001
 soundClassObject=/Game/Sounds/FortSoundClassesAndModes/Fort_Backblings
 volumeMultiplier=1
 concurrencyOverrides {bLimitToOwner=true, maxCount=1, resolutionRule=StopOldest}
}
